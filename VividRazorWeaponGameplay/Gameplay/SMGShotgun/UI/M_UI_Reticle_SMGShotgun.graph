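# Translucent UI material that draws the SMG shotgun crosshair procedurally from signed distance fields, with spread-driven animation.
# Four radially placed box prongs move with spread, around bracket outlines made by intersecting a square and an octagon that grow with bloom.
# Radially mirrored boxes cut gaps in the brackets, and a small center box completes the union.
# Blending an antialiased edge with a reversed smoothstep softens the final mask.
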Material /VividRazorWeaponGameplay/Gameplay/SMGShotgun/UI/M_UI_Reticle_SMGShotgun
graph {
 properties {bAutomaticallySetUsageInEditor=false, blendMode=BLEND_Translucent, materialDomain=MD_UI, usedShadingModels=Default Lit}
 nodes [
  MaterialFunctionCall() {materialFunction=&RemapValueRange}
  n1 = MaterialFunctionCall(
   UVs=MaterialFunctionCall!(
    Reticle UVs
    Reticle Sides=Constant() {r=4}
    Reticle Scale
    Reticle Rotation=Constant() {r=0.125}
    Prong Scale
    Prong Rotation
    Spread Offset=ScalarParameter() {defaultValue=4, parameterName=ProngSpreadOffset}
    Positional Offset=Constant() {r=-25}
    "Spread Input Range [Min,Max]"=Constant2Vector() {g=40}
    "Spread Remapped Range [Min,Max]"=Constant2Vector() {g=-0.1, r=-0.32}
   ) {materialFunction=/FortUILibrary/Systems/Reticles/MaterialFunctions/MF_UI_Reticle_UVs}.UVs
   Size=AppendVector(
    A=ScalarParameter() {defaultValue=0.0175, group=Shape Operators, parameterName=BoxWidth, sliderMax=1}
    B=ScalarParameter() {defaultValue=0.05, group=Shape Operators, parameterName=BoxHeight, sliderMax=1}
   )
  ) {materialFunction=&MF_UI_Box}
  n2 = Constant() {r=0.125}
  n3 = Rotator(Time=@n2) {speed=6.2832}
  n4 = NamedRerouteDeclaration(None=MaterialFunctionCall!(
    Positional Spread Offset=Constant() {r=-10}
    Accuracy Spread Offset=ScalarParameter() {defaultValue=47, group=Reticle Operators, parameterName=BracketSpreadOffset}
    Min Value
    "Input Range [Min,Max]"=Constant2Vector() {g=100}
    "Remapped Range [Min,Max]"=Constant2Vector() {g=1, r=-0.4}
   ) {materialFunction=/FortUILibrary/Systems/Reticles/MaterialFunctions/MF_UI_Reticle_Remap_Spread}.Result) {name=Bracket Spread/Bloom, nodeColor=<rgba 0 0 0 1>}
  n5 = ScalarParameter() {defaultValue=0.3, parameterName=Reticle Size, sliderMax=1}
  n6 = Add(A=@n5, B=@n4) {constB=0}
  n7 = Constant() {r=4}
  n8 = Constant() {r=-1}
  n9 = Rotator(Time=@n2) {speed=3.1416}
  n10 = Add(A=@n4, B=Multiply(A=@n5) {constB=0.98})
  n11 = Constant() {r=8}
  n12 = MaterialFunctionCall!(
   DistanceField=Max(
    A=MaterialFunctionCall!(UVs=@n3, Size=@n6, Sides=@n7, InnerAngle, OutlineThickness, OutlineAlignment=@n8, GlowMin, GlowMax) {materialFunction=/FortUILibrary/MaterialFunctions/Shapes/MF_UI_RegularPolygon}.SDF
    B=MaterialFunctionCall(UVs=@n9, Size=@n10, Sides=@n11, OutlineAlignment=@n8) {materialFunction=&MF_UI_RegularPolygon}.SDF
   )
   Threshold
   Thickness=ScalarParameter() {defaultValue=0.009, parameterName=OutlineThickness}
   Alignment=Constant() {r=-1}
   Smoothing
  ) {materialFunction=/FortUILibrary/MaterialFunctions/Shapes/Modifiers/MF_UI_ShapeOutline}
  n13 = MaterialFunctionCall() {materialFunction=&GetUserInterfaceUV}
  n14 = ScalarParameter() {defaultValue=0.125, group=Gap Operators, parameterName=GapRotation}
  n15 = ScalarParameter() {defaultValue=0.3, group=Gap Operators, parameterName=GapWidth, sliderMax=1}
  n16 = NamedRerouteUsage()
  n17 = MaterialFunctionCall(
   UVs=MaterialFunctionCall!(
    UVs=MaterialFunctionCall(UVs=@n13.Normalized UV, "Rotation Angle (0-1)"=@n14) {materialFunction=&CustomRotator}.Rotated Values
    Sides=ScalarParameter() {defaultValue=4, group=Gap Operators, parameterName=GapSections, sliderMax=12}
   ) {materialFunction=/FortUILibrary/MaterialFunctions/Basics/MF_UI_UVRadialMirror}.UVs
   Size=AppendVector(A=Add(A=@n15, B=@n16) {constB=0}, B=Constant() {r=1})
  ) {materialFunction=&MF_UI_Box}
  n18 = MaterialFunctionCall!(
   DistanceFieldA=MaterialFunctionCall(
    DistanceFieldA=Multiply(A=@n1.SDF) {constB=0.6, desc=For Dropshadow}
    DistanceFieldB=MaterialFunctionCall!(DistanceFieldA=@n12.SDFOutline, DistanceFieldB=@n17.SDF, BlendAmount) {materialFunction=/FortUILibrary/MaterialFunctions/Shapes/Modifiers/MF_UI_SDF_Subtraction}.Exact
   ) {materialFunction=/FortUILibrary/MaterialFunctions/Shapes/Modifiers/MF_UI_SDF_Union}.Exact
   DistanceFieldB=MaterialFunctionCall!(UVs, Size=ScalarParameter() {defaultValue=0.009, parameterName=OutlineThickness}, CornerRoundness, KeepAspectRatio, PixelSize) {materialFunction=/FortUILibrary/MaterialFunctions/Shapes/SDFs/MF_SDF_Box}.SDF
   BlendAmount
  ) {materialFunction=&MF_UI_SDF_Union}
  Multiply(A=LinearInterpolate(
    A=MaterialFunctionCall(DistanceField=@n18.Exact) {materialFunction=&MF_UI_AntiAliasing}.Inner
    B=SmoothStep(Value=@n18.Exact) {constMax=0.001, constMin=0.0125, desc=Min and Max reversed due to SDF}
   ) {constAlpha=0.4}) {constB=0.95}
 ]
}
