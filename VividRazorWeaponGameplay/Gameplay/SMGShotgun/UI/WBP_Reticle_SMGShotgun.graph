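# Widget Blueprint reticle for the SMG/shotgun dual weapon, built on the shared dual-weapon reticle base, with one Image using the SMGShotgun reticle material.
# On initialization it caches that image's dynamic material instance.
# Each tick, for each valid weapon, it scales spread-to-field-of-view ratio by a constant into a material scalar.
# The primary weapon drives the bracket spread offset and the offhand drives the prong offset.
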WidgetBlueprint /VividRazorWeaponGameplay/Gameplay/SMGShotgun/UI/WBP_Reticle_SMGShotgun
graphs [EventGraph {nodes [
   Event Tick()
   > n1 = Sequence()
   n2 = Get varPrimaryWeapon() -> (varPrimaryWeapon:Fort Weapon Ranged Object Reference)
   > Is Valid(InputObject=@n2)
   n3 = Get owningFortPlayerController() -> (owningFortPlayerController:Fort Player Controller Object Reference)
   > SetScalarParameterValue(
    self=Get varDynamicReticle() -> (varDynamicReticle:Material Instance Dynamic Object Reference)
    ParameterName=BracketSpreadOffset
    Value=float * float(
     A=GetSpreadToFieldOfViewRatio!(self:Fort Weapon Ranged Object Reference=@n2, PlayerController:Fort Player Controller Object Reference=@n3, bUseMinimumSpread:bool=false) -> (ReturnValue:float)
     B=1000
    ) {is=title K2Node_PromotableOperator}
   )
   n4 = Get varOffhandWeapon() -> (varOffhandWeapon:Fort Weapon Ranged Object Reference)
   Is Valid(exec=@n1.then_1, InputObject=@n4)
   > SetScalarParameterValue(
    self=Get varDynamicReticle()
    ParameterName=ProngSpreadOffset
    Value=float * float(A=GetSpreadToFieldOfViewRatio(self=@n4, PlayerController=@n3, bUseMinimumSpread=false), B=1000)
   )
   Event OnInitialized()
   > Parent: OnInitialized() {is=title K2Node_CallParentFunction}
   > n5 = GetDynamicMaterial(self=Get Reticle() -> (Reticle:Image Object Reference))
   > Set varDynamicReticle(varDynamicReticle:Material Instance Dynamic Object Reference=@n5) -> (Output_Get:Material Instance Dynamic Object Reference)
  ]}]
variables [varDynamicReticle {type=MaterialInstanceDynamic}]
widgets {
 info {
  inheritedWidgetCount=0
  namedSlotCount=1
  parentClass=/BattleRoyaleDualWeaponUI/Widgets/WBP_Reticle_DualWeaponBase.WBP_Reticle_DualWeaponBase_C
  rootWidgetClass=null
  widgetCount=1
 }
 widgets [Reticle {
   widgetClassPath=&Image
   bIsVariable=true
   widget=Reticle
   properties {brush {
     bIsDynamicallyLoaded=false
     drawAs=Image
     imageType=FullColor
     mirroring=NoMirror
     resourceName=None
     resourceObject=&/M_UI_Reticle_SMGShotgun
     tiling=NoTile
     imageSize=<256 256>
     margin=<ltrb 0 0 0 0>
     outlineSettings {
      bUseBrushTransparency=false
      roundingType=HalfHeightRadius
      width=0
      color {colorUseRule=UseColor_Specified, specifiedColor=<rgba 0 0 0 0>}
      cornerRadii=<0 0 0 0>
     }
     tintColor {colorUseRule=UseColor_Specified, specifiedColor=<rgba 1 1 1 1>}
     uVRegion {bIsValid=false, max=<0 0>, min=<0 0>}
    }}
  }]
}
